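# Additive unlit mesh-particle material for a glow effect on the Bounty Hunter pickaxe.
# Particle color is boosted through an emissive adjustment function, and a scaled vertex-normal branch suggests slight mesh inflation.
# Time-scrolled second-channel UVs, taken unrotated and at 60 and 120 degrees, index curve-atlas masks, forming three summed, repeating band sets without texture maps.
# That pattern is multiplied by Fresnel, depth-faded and scaled by particle alpha.
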Material /BRCosmetics/Weapons/FORT_Melee/Pickaxe_BountyHunter/FX/M_Pickaxe_BountyHunter_Glow
graph {
 properties {bUsedWithMeshParticles=true, blendMode=BLEND_Additive, pixelDepthOffsetMode=PDOM_Legacy, shadingModel=MSM_Unlit, usedShadingModels=Unlit}
 nodes [
  n1 = ParticleColor()
  MaterialFunctionCall(Emissive=Multiply(A=Constant() {r=3}, B=@n1.RGB)) {materialFunction=&MF_FX_EmissiveAdjustments}
  Multiply(A=VertexNormalWS(), B=Constant3Vector() {constant=<rgba 0.2 0.2 0.2 0>})
  n2 = Fresnel(ExponentIn=Constant() {r=3.2}, BaseReflectFractionIn=Constant())
  n3 = VectorParameter() {defaultValue=<rgba 21 3 8 16>}
  n4 = Add(A=Multiply(A=TextureCoordinate() {coordinateIndex=1}, B=AppendVector(A=@n3.R, B=@n3.R)), B=Multiply(A=Time()) {constB=0.25})
  n5 = MaterialFunctionCall(UVs=@n4, "Rotation Angle (0-1)"=Constant() {r=0.3333}) {materialFunction=&CustomRotator}
  n6 = CurveAtlasRowParameter(CurveTime=ComponentMask(None=Frac(None=@n5.Rotated Values)) {g=true}) {
   atlas=&AthenaCurveAtlas01
   curve=/Game/Athena/LinearColorCurves/Curve_Masks
   defaultValue=67
   parameterName=Param_2
  }
  n7 = MaterialFunctionCall(UVs=@n4, "Rotation Angle (0-1)"=Constant() {r=0.1667}) {materialFunction=&CustomRotator}
  n8 = CurveAtlasRowParameter(CurveTime=ComponentMask(None=Frac(None=@n7.Rotated Values)) {g=true}) {
   atlas=&AthenaCurveAtlas01
   curve=&Curve_Masks
   defaultValue=67
   parameterName=Param_3
  }
  n9 = CurveAtlasRowParameter(CurveTime=ComponentMask(None=Frac(None=@n4)) {g=true}) {
   atlas=&AthenaCurveAtlas01
   curve=&Curve_Masks
   defaultValue=67
   parameterName=Param_4
  }
  n10 = Multiply(A=Add(A=Add(A=@n6.G, B=@n8.G), B=@n9.G)) {constB=50}
  Multiply(A=@n1.A, B=DepthFade(Opacity=Multiply(A=@n2, B=@n10)))
 ]
}
